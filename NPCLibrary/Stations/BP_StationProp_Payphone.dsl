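; Payphone NPC station prop that plays Season 17 Welcome Party quest voice lines for the local player, chosen by conversation message number.
; Local dialog-marker state changes drive StartRinging and StopRinging, with a looping timer replaying the ambient ring until service use or destruction.
; As written, StartRinging sets its local quest-eligibility flag false before testing it, so the timer never starts.
; The wobble-stop event is empty.
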(event EventBeginPlay
  (|CallParentFunction)
  (Utilities|FlowControl|Sequence
    (:then_0
      ;[1 1 1] Cache off intitial relative rotation so wobble timeline doesn't stomp any data
      (Networking|SwitchHasAuthority
        (:Remote
          (Variables|Default|SetInitial%20Relative%20Rotation (Variables|/Script/Engine.SceneComponent|GetRelativeRotation (Variables|Debug|GetStaticMeshComponent))))))
    (:then_1
      ;[1 1 1] Initialize bindings so looping Ring Audio + Wobble timeline can be kept in sync with *local* user interaction marker state
      (EventDispatchers|BindEventtoOnDialogMarkerInteractionStateChanged (Variables|Components|GetNonPlayerConversationComponent) (delegate Custom|LocalUserInteractionStateChanged)))))

(event Custom|LocalUserInteractionStateChanged
  (params (CurrentInteractionState :type EDialogMarkerInteractionState%20Enum :default "Conversation") (InteractingController :type Fort%20Player%20Controller%20Object%20Reference))
  ;[1 1 1] Initialize bindings so looping Ring Audio + Wobble timeline can be kept in sync with *local* user interaction marker state
  (Utilities|FlowControl|Sequence
    (:then_0
      (switch Utilities|FlowControl|Switch|SwitchonEDialogMarkerInteractionState CurrentInteractionState
        (:Conversation
          _)
        (:InteractionRange
          (if (or (Utilities|Enum|Equal(Enum) (Variables|Default|GetLastDialogMarkerInteractionState) "Conversation") (Utilities|Enum|Equal(Enum) (Variables|Default|GetLastDialogMarkerInteractionState) "None")) :!label _label
            (CallFunction|StartRinging)))
        (:IndicatorRange
          (goto _label))
        (:None
          _))
      (CallFunction|StopRinging))
    (:then_1
      (Variables|Default|SetLastDialogMarkerInteractionState CurrentInteractionState))))

(event Custom|PlayQuestVOs
  (params (VONumber :type Integer) (FPC :type Fort%20Player%20Controller%20Object%20Reference))
  ;[1 1 1] S17 Welcome Party Payphone Quest VO Segments
  (Networking|SwitchHasAuthority
    (:Remote
      (if (Pawn|IsLocalPlayerController FPC)
        (switch int VONumber
          (:3
            ;;[1 1 1] Week 3 - Believer Beach, Find "Joey"
            (Concatenator|Play (Variables|Default|GetQuestVO_01) (Variables|Default|GetPayphone_QuestVO)))
          (:5
            ;;[1 1 1] Week 5 - "Sunny", Listen to Hot Saucers
            (Concatenator|Play (Variables|Default|GetQuestVO_02) (Variables|Default|GetPayphone_QuestVO)))
          (:6
            ;;[1 1 1] Week 6 - Hayseed's Farm, Bribe Him for Access to Farm
            (Concatenator|Play (Variables|Default|GetQuestVO_03) (Variables|Default|GetPayphone_QuestVO)))
          (:8
            ;;[1 1 1] 17.20 - Find a Field Team Access Card
            (Concatenator|Play (Variables|Default|GetQuestVO_04) (Variables|Default|GetPayphone_QuestVO)))
          (:9
            ;;[1 1 1] Look into the Strange Billboards
            (Concatenator|Play (Variables|Default|GetQuestVO_12) (Variables|Default|GetPayphone_QuestVO)))
          (:10
            ;;[1 1 1] Learn everything you can about the abductions
            (Concatenator|Play (Variables|Default|GetQuestVO_13) (Variables|Default|GetPayphone_QuestVO)))
          (:11
            ;;[1 1 1] Squeeze intel out of alien in "deep cover"
            (Concatenator|Play (Variables|Default|GetQuestVO_14) (Variables|Default|GetPayphone_QuestVO)))
          (:12
            ;;[1 1 1] Weapons Testing
            (Concatenator|Play (Variables|Default|GetQuestVO_15) (Variables|Default|GetPayphone_QuestVO)))
          (:13
            ;;[1 1 1] Warn people before mothership crashes
            (Concatenator|Play (Variables|Default|GetQuestVO_22) (Variables|Default|GetPayphone_QuestVO))))))))

(event Custom|TriggerNextRing
  ;[1 1 1] Wobble timeline control
  (if (not (Variables|BuildingActor|GetbDestroyed))
    (Audio|Components|Audio|Play (Variables|Components|GetStation_Ambient) 0.0)))

(event Custom|StopRingWobbleTimeline)

(event Collision|EventActorBeginOverlap
  (params (OtherActor :type Actor%20Object%20Reference))
  (|CallParentFunction OtherActor) ; This node is disabled and will not be called.\nDrag off pins to build functionality.
)

(event EventTick
  (params (DeltaSeconds :type Float%20%28single-precision%29))
  (|CallParentFunction DeltaSeconds) ; This node is disabled and will not be called.\nDrag off pins to build functionality.
)

(event EventOnDeathPlayEffects
  (params (Damage :type Float%20%28single-precision%29) (DamageTags :type Gameplay%20Tag%20Container%20Structure%20%28by%20ref%29) (Momentum :type Vector :default "0, 0, 0") (HitInfo :type Hit%20Result%20Structure%20%28by%20ref%29) (InstigatedBy :type Fort%20Pawn%20Object%20Reference) (DamageCauser :type Actor%20Object%20Reference) (EffectContext :type Gameplay%20Effect%20Context%20Handle%20Structure))
  (CallFunction|StopRinging))

(fn BlueprintOnLocalInteract
  (params (InteractingPawn :type Fort%20Player%20Pawn%20Object%20Reference) (InteractionBeingAttempted :type EInteractionBeingAttempted%20Enum :default "FirstInteraction"))
  (Audio|PlaySoundatLocation (Variables|Default|GetInteractSound) (Transformation|GetWorldLocation (Variables|Components|GetStation_Ambient)))
  (return false))

(fn ConversationMessage
  (params (Message%20Number :type Integer) (FPC :type Fort%20Player%20Controller%20Object%20Reference))
  (CallFunction|PlayQuestVOs Message%20Number FPC)
  (Development|PrintString (Utilities|String|Append "MessageShouldPlay
" (Utilities|String|ToString(Integer) Message%20Number)))
  (return))

(fn PlayServiceUsedEffects
  (params (PlayerController :type Fort%20Player%20Controller%20Object%20Reference))
  (|CallParentFunction PlayerController)
  (Variables|Default|SetbLocalPlayerHasUsedServices true))

(fn StartRinging
  (Utilities|FlowControl|Sequence
    (:then_0
      (LocalVariables|SetLocalPlayerCanReceiveQuest false))
    (:then_1
      ;[1 1 1] Once quest accepted, this payphone cannot be used again. Stop signposting
      (if (Variables|Default|GetbLocalPlayerHasUsedServices)
        _
        (elif (LocalVariables|GetLocalPlayerCanReceiveQuest)
          (if (not (Variables|BuildingActor|GetbDestroyed))
            (bind _returnvalue (Utilities|Time|SetTimerbyEvent (EventDispatchers|CreateEvent :!event TriggerNextRing) (+ (Variables|/Script/Engine.SoundBase|GetDuration (Variables|/Script/Engine.AudioComponent|GetSound (Variables|Components|GetStation_Ambient))) (Variables|Default|GetAdditionalDelayBetweenRings)) true))
            (Variables|Default|SetRingTimerHandle _returnvalue)))))))

(fn StopRinging
  (if (Utilities|Time|IsValidTimerHandle (Variables|Default|GetRingTimerHandle))
    (Audio|Components|Audio|Stop (Variables|Components|GetStation_Ambient))
    (CallFunction|StopRingWobbleTimeline)
    (Utilities|Time|ClearandInvalidateTimerbyHandle (Variables|Default|GetRingTimerHandle))))

(fn ConstructionScript
  (|CallParentFunction))
