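; NPC station payphone prop that randomly decides at spawn whether it can ring and be answered.
; The server rolls RingChance into replicated AllowInteractRep; clients read it after a fixed delay to show or hide the icon, as its OnRep is empty.
; Local dialog marker state changes drive a looping timer that replays the ring audio.
; Answering plays a click, hides the icon and locks out that local player; destruction stops ringing.
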(event EventBeginPlay
  ;[0 0 0] Determine if the payphone should ring
  (|CallParentFunction)
  (Networking|SwitchHasAuthority
    (:Authority
      (if (<= (Math|Random|RandomIntegerinRange 0 100) (Variables|Default|GetRingChance))
        (Variables|Default|SetAllowInteractRep true)
        (Development|PrintString "Payphone can  ring" true true "(R=1.000000,G=0.000000,B=0.288397,A=1.000000)" 7.0)
        (else
          (Variables|Default|SetAllowInteractRep false)
          (Development|PrintString "Payphone cannot ring" true true "(R=1.000000,G=0.000000,B=0.288397,A=1.000000)" 7.0))))
    (:Remote
      (Utilities|FlowControl|Delay 4.0)
      (if (Variables|Default|GetAllowInteractRep)
        (Development|SetHiddeninGame (Variables|Default|GetPayphoneIconBillboard) false false)
        (else
          (Development|SetHiddeninGame (Variables|Default|GetPayphoneIconBillboard) true false))))))

(event OnDialogMarkerInteractionStateChanged(NonPlayerConversationComponent)
  (params (CurrentInteractionState :type EDialogMarkerInteractionState%20Enum :default "Conversation") (InteractingController :type Fort%20Player%20Controller%20Object%20Reference))
  ;[0 0 0] Keep Ring Audio in sync with local user interaction marker state
  (if (Variables|Default|GetAllowInteractRep)
    (Utilities|FlowControl|Sequence
      (:then_0
        (switch Utilities|FlowControl|Switch|SwitchonEDialogMarkerInteractionState CurrentInteractionState
          (:Conversation
            _)
          (:InteractionRange
            (if (or (Utilities|Enum|Equal(Enum) (Variables|Default|GetPreviousInteractionState) "Conversation") (Utilities|Enum|Equal(Enum) (Variables|Default|GetPreviousInteractionState) "None")) :!label _label
              (CallFunction|StartRinging)))
          (:IndicatorRange
            (goto _label))
          (:None
            _))
        (CallFunction|StopRinging))
      (:then_1
        (Variables|Default|SetPreviousInteractionState CurrentInteractionState)))))

(event Custom|TriggerNextRing
  ;[0 0 0] Play subsequent rings
  (if (not (Variables|BuildingActor|GetbDestroyed))
    (Audio|Components|Audio|Play (Variables|Components|GetStation_Ambient) 0.0)))

(event EventOnDeathPlayEffects
  (params (Damage :type Float%20%28single-precision%29) (DamageTags :type Gameplay%20Tag%20Container%20Structure%20%28by%20ref%29) (Momentum :type Vector :default "0, 0, 0") (HitInfo :type Hit%20Result%20Structure%20%28by%20ref%29) (InstigatedBy :type Fort%20Pawn%20Object%20Reference) (DamageCauser :type Actor%20Object%20Reference) (EffectContext :type Gameplay%20Effect%20Context%20Handle%20Structure))
  (CallFunction|StopRinging))

(fn BlueprintCanInteract
  (params (InteractingPawn :type Fort%20Pawn%20Object%20Reference) (InteractionBeingAttempted :type EInteractionBeingAttempted%20Enum :default "FirstInteraction") (InteractionType :type TInteractionType%20Enum :default "IT_NoInteraction"))
  (if (Math|Boolean|ANDBoolean (not (Variables|Default|GetLocal%20Player%20Has%20Used)) (not (Variables|BuildingActor|GetbDestroyed)) (Variables|Default|GetAllowInteractRep))
    (bind _returnvalue (Conversation|FindConversationComponent (Pawn|GetController InteractingPawn)))
    (Utilities|IsValid _returnvalue
      (:Is%20Valid
        (return (not (Conversation|IsinActiveConversation _returnvalue))))
      (:Is%20Not%20Valid
        (return true)))
    (else
      (return false))))

(fn BlueprintGetInteractionString
  (params (InteractingPawn :type Fort%20Pawn%20Object%20Reference) (InteractionBeingAttempted :type EInteractionBeingAttempted%20Enum :default "FirstInteraction"))
  (return "NSLOCTEXT(\"[6E74FD6F0288200E5B34F496ABA5B65E]\", \"6C75A47342AD41FCC918B990B63398BC\", \"Answer\")"))

(fn BlueprintOnLocalInteract
  (params (InteractingPawn :type Fort%20Player%20Pawn%20Object%20Reference) (InteractionBeingAttempted :type EInteractionBeingAttempted%20Enum :default "FirstInteraction"))
  ;[0 0 0] When interacted with, play the ring click sound
  (Audio|PlaySoundatLocation (Variables|Default|GetRingInteractCue) (Transformation|GetWorldLocation (Variables|Components|GetStation_Ambient)))
  (Variables|Default|SetLocal%20Player%20Has%20Used true)
  (Development|SetHiddeninGame (Variables|Default|GetPayphoneIconBillboard) true false)
  (return false))

(fn OnRep_AllowInteractRep)

(fn StartRinging
  ;[0 0 0] Start ringing the payphone and save off the ring timer if it hasn't been used or destroyed
  (if (Variables|Default|GetLocal%20Player%20Has%20Used)
    _
    (elif (Variables|BuildingActor|GetbDestroyed)
      _
      (else
        (bind _returnvalue (Utilities|Time|SetTimerbyEvent (EventDispatchers|CreateEvent :!event TriggerNextRing) (+ (Variables|/Script/Engine.SoundBase|GetDuration (Variables|/Script/Engine.AudioComponent|GetSound (Variables|Components|GetStation_Ambient))) (Variables|Default|GetAdditionalDelayBetweenRings)) true))
        (Variables|Default|SetRingTimerHandle _returnvalue)))))

(fn StopRinging
  ;[0 0 0] Stop ringing the payphone; clear the ringing timer
  (if (Utilities|Time|IsValidTimerHandle (Variables|Default|GetRingTimerHandle))
    (Audio|Components|Audio|Stop (Variables|Components|GetStation_Ambient))
    (Utilities|Time|ClearandInvalidateTimerbyHandle (Variables|Default|GetRingTimerHandle))))

(fn ConstructionScript
  (|CallParentFunction)
  (Audio|Components|Audio|SetSound (Variables|Components|GetStation_Ambient) (Variables|Default|GetRingAmbientCue)))
