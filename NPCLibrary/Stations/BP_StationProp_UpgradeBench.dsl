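; Upgrade Bench station prop Blueprint that gives audio-visual feedback when its service is used.
; PlayServiceUsedEffects triggers TriggerStationUsedEffects, which plays a success sound, activates upgrade FX and wobbles the static mesh around its cached initial rotation via a timeline.
; On BeginPlay, authority destroys the bench when a scalable float disallows it in the current mode.
; EndPlay carries temporary, flag-gated shipping logging to investigate benches being deleted in Creative Royale.
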(event EventEndPlay
  (params (EndPlayReason :type EEndPlayReason%20Enum :default "Destroyed"))
  ;[1 0.7623 0.218] Log Debug info to help identify cause of Upgrade Benches being deleted in Creative Royale Edit/Play mode.
  ;
  ; This is TEMPORARY and will be removed before public release.
  (|CallParentFunction EndPlayReason)
  (if (ScalableFloat|GetValueasBool (Variables|Default|GetCanPostShippingLogs))
    (Fortnite|FortShippingLog self (Utilities|String|Append "EndPlay caused by " (Utilities|String|EnumtoString EndPlayReason)) true)))

(event EventBeginPlay
  (|CallParentFunction)
  (Variables|Default|SetInitialRelativeRotation (Variables|/Script/Engine.SceneComponent|GetRelativeRotation (Variables|Debug|GetStaticMeshComponent)))
  ;[0.281 1 0.4092] Destroy the vending machine if in Creative Royale.
  (Networking|SwitchHasAuthority
    (:Authority
      (if (not (ScalableFloat|GetValueasBool (Variables|Default|GetAllowedInMode)))
        (Actor|DestroyActor)))))

(event Collision|EventActorBeginOverlap
  (params (OtherActor :type Actor%20Object%20Reference))
  (|CallParentFunction OtherActor) ; This node is disabled and will not be called.\nDrag off pins to build functionality.
)

(event EventTick
  (params (DeltaSeconds :type Float%20%28single-precision%29))
  (|CallParentFunction DeltaSeconds) ; This node is disabled and will not be called.\nDrag off pins to build functionality.
)

(event Custom|TriggerStationUsedEffects
  (bind _p_upgradefx (Variables|Default|GetP_UpgradeFX))
  (Utilities|FlowControl|Sequence
    (:then_0
      (Audio|PlaySoundatLocation (Variables|Default|GetServiceSuccessSound) (Transformation|GetActorLocation)))
    (:then_1
      (Utilities|IsValid _p_upgradefx
        (:Is%20Valid
          (Components|Activation|SetActive _p_upgradefx true true))))
    (:then_2
      (bind (_direction _wobbleovertime) (Timeline|WobbleTimeline PlayFromStart)
        (:Update
          (Transformation|SetRelativeRotation (Variables|Debug|GetStaticMeshComponent) (Math|Rotator|CombineRotators (Variables|Default|GetInitialRelativeRotation) (Math|Rotator|Lerp(Rotator) "0, 0, 0" "0.500000,0.000000,0.000000" _wobbleovertime true)) false false))))))

(fn PlayServiceUsedEffects
  (params (PlayerController :type Fort%20Player%20Controller%20Object%20Reference))
  (CallFunction|TriggerStationUsedEffects))

(fn ConstructionScript
  (|CallParentFunction))
